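# Niagara effect for the Holiday Lights backpack cosmetic, simulating swirling snow-globe flakes and a warm star glow on its skeletal mesh.
# CPU and GPU snow emitters confine sprites to an owner-scaled sphere around sampled mesh points, the GPU one adding a denser, finer spawn group.
# Owner speed intensifies vortex and smoothed random motion, while animated sprite Y-scale fakes tumbling.
# StarLights holds a bone-attached sprite whose alpha follows time of day.
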NiagaraSystem /BRCosmetics/Accessories/FORT_Backpacks/M_MED_Holiday_Lights_Backpack/Effects/NS_Backpack_HolidayLights
dependencies {
 usedDataInterfaces [&NiagaraDataInterfaceColorCurve, &NiagaraDataInterfaceCurve]
 usedDynamicInputs []
 usedModules [&SystemState]
}
emitters {
 SnowGlobe {
  emitterUpdateScript [
   &EmitterState()
   &SpawnRate(SpawnRate=5, Use Spawn Probability=false)
   /SetVariables(Emitter.VelocityScale:NiagaraFloat=&Multiply_Float(A=0.015, B=&VectorLength(Vector=$Engine.Owner.Velocity))) {isSetParameters=true, module=SetVariables_#1}
  ]
  particleSpawnScript [
   /Niagara/Modules/Spawn/Initialization/InitializeParticle(
    Color=&MakeLinearColorFromVectorAndFloat("Vector (RGB)"=<3 3 3>, "Float (Alpha)"=&RandomRangeFloat(Minimum=0.1, Maximum=1))
    Lifetime=&RandomRangeFloat(Minimum=2, Maximum=6)
    Module.Write Color=true
    Module.Write Lifetime=true
    Module.Write Mass=false
    Module.Write Material Random=true
    Module.Write Position=false
    Module.Write Scale=false
    Module.Write SpriteRotation=true
    Module.Write SpriteSize=true
    Position=&SimulationPosition()
    Sprite Rotation=&RandomRangeFloat(Minimum=-360, Maximum=360)
    Sprite Size=&Vector2DFromFloat(Value=&RandomRangeFloat(Minimum=0.5, Maximum=3))
   )
   &SkeletalMeshLocation(
    Inherit Velocity=false
    MaskSpawnGroup=false
    Mesh Sampling Type=::NewEnumerator5
    Sampled Position Offset=&RandomRangeVector(Minimum=<-9 -9 -9>, Maximum=<9 9 9>)
    Sampled Position Offset Coordinate Space:ENiagaraMeshTransforms=::NewEnumerator1
    Skeletal Mesh=$User.SkeletalMesh
    Socket Sampling Mode=::NewEnumerator5
   )
  ]
  particleUpdateScript [
   &ParticleState()
   &SkeletalMeshLocation(
    Apply Sampled Orientation to Mesh Particles=true
    Apply To Particle Position=false
    Inherit Velocity=false
    Interpolate Mesh Sampling=false
    MaskSpawnGroup=false
    Mesh Sampling Type=::NewEnumerator5
    Sampled Position Offset Coordinate Space:ENiagaraMeshTransforms=::NewEnumerator1
    Skeletal Mesh=$User.SkeletalMesh
    Socket Sampling Mode=::NewEnumerator5
   ) {module=SkeletalMeshLocation001}
   /Niagara/Modules/Update/Velocity/VortexVelocity(
    Module.Use Influence Falloff=false
    Velocity Amount:NiagaraFloat=&self:InterpolateFloat(
     Float to Interpolate:NiagaraFloat=&Lerp_Float(
      A=&RandomRangeFloat(Minimum=1, Maximum=3)
      B=&RandomRangeFloat(Minimum=10, Maximum=200)
      Alpha=&ClampFloat(Float=$Emitter.VelocityScale)
     )
     Initial Value:NiagaraFloat=2
     Rate Of Change:NiagaraFloat=3
    )
    Vortex Axis:Vector3f=&RandomRangeVector(Minimum=<-0.75 -0.75 1>, Maximum=<0.75 0.75 1>)
    Vortex Origin:Vector3f=$Output.SkeletalMeshLocation001.SampledPosition
   )
   &self:AddSmoothRandomVelocity(
    Rate Of Change=5
    Velocity:Vector3f=&Multiply_VectorByFloat(
     Vector=&Lerp_Vector(
      A=&RandomRangeVector(Minimum=<-5 -5 -6>, Maximum=<5 5 3>, Evaluation Type=::NewEnumerator1)
      B=&RandomRangeVector(Minimum=<-30 -30 -30>, Maximum=<30 30 30>, Evaluation Type=::NewEnumerator1)
      Alpha=&ClampFloat(Float=$Emitter.VelocityScale, Max=50)
     )
     Float=&ClampFloat(Float=&Multiply_Float(A=3, B=$Particles.NormalizedAge))
    )
   ) {module=AddSmoothRandomVelocity}
   &SolveForcesAndVelocity(Clamp Velocity=false, Limit Acceleration=false)
   &self:ConstrainToSphere(Sphere Radius:NiagaraFloat=&Multiply_Float(A=9.5, B=&MakeFloatFromVector(Vector=$Engine.Owner.Scale, Channel=::NewEnumerator0))) {module=ConstrainToSphere}
   /SetVariables(Particles.SpriteRotation:NiagaraFloat=&Add_Float(
     A=$Particles.Initial.SpriteRotation
     B=&Lerp_Float(
      A=0
      B=&Multiply_Float(
       A=&RandomRangeFloat(Minimum=-2, Maximum=-1)
       B=&RandomRangeFloat(Minimum=150, Maximum=360)
      )
      Alpha=$Particles.NormalizedAge
     )
    )) {isSetParameters=true, module=SetVariables_#2}
   &Color(Color=&MultiplyLinearColor(
     Linear Color=$Particles.Initial.Color
     Scale Factor=&MakeVector4_FromLinearColor(LinearColor=&ColorFromCurve(
       ColorCurve=DataInterface() {propertyValues {
         AlphaCurve {keys [<at 0 1 RCIM_Cubic>]}
         BlueCurve {keys [<at 0 0.16 RCIM_Cubic>,<at 0.4999 1>,<at 1 1 RCIM_Cubic>]}
         GreenCurve {keys [<at 0 1 RCIM_Cubic>,<at 0.4999 1>,<at 1 0.1 RCIM_Cubic>]}
         RedCurve {keys [<at 0 0.1 RCIM_Cubic>,<at 0.4999 1>,<at 1 0.1 RCIM_Cubic>]}
        }}
       CurveIndex=&ClampFloat(Float=&self:InterpolateFloat(
         Float to Interpolate:NiagaraFloat=&RandomRangeFloat(Minimum=-2, Maximum=2.5, Evaluation Type=::NewEnumerator1)
         Initial Value:NiagaraFloat=0.5
         Rate Of Change:NiagaraFloat=&RandomRangeFloat(Minimum=2, Maximum=2, Evaluation Type=::NewEnumerator1)
        ))
      ))
    )) {enabled=false, module=Color001}
   &ScaleSpriteSize(Scale Factor=&Multiply_Vector2(
     A=&MakeVector2D(
      X=1
      Y=&Lerp_Float(
       A=-1
       B=2
       Alpha=&self:InterpolateFloat(
        Float to Interpolate:NiagaraFloat=&RandomRangeFloat(Minimum=0, Maximum=1, Evaluation Type=::NewEnumerator1)
        Initial Value:NiagaraFloat=&RandomRangeFloat(Minimum=0.5, Maximum=1)
        Rate Of Change:NiagaraFloat=&RandomRangeFloat(Minimum=0.25, Maximum=0.75, Evaluation Type=::NewEnumerator1)
       )
      )
     )
     B=&Vector2DFromFloat(Value=&FloatFromCurve(FloatCurve=DataInterface() {propertyValues {Curve {keys [
           {arriveTangent=4.9809, interpMode=RCIM_Cubic, leaveTangent=4.9809, tangentMode=RCTM_User, time=0, value=0}
           {arriveTangent=0.9768, interpMode=RCIM_Cubic, leaveTangent=0.9768, tangentMode=RCTM_User, time=0.35, value=1}
           {arriveTangent=-1.0878, interpMode=RCIM_Cubic, leaveTangent=-1.0878, tangentMode=RCTM_User, time=0.65, value=1}
           {arriveTangent=-6.811, interpMode=RCIM_Cubic, leaveTangent=-6.811, tangentMode=RCTM_User, time=1, value=0}
          ]}}}))
    ))
   &Color(Color=&ColorFromCurve(
     ColorCurve=DataInterface() {propertyValues {
       AlphaCurve {keys [<at 0 1 RCIM_Cubic>]}
       BlueCurve {keys [<at 0 0 RCIM_Cubic>,<at 0.5 0>,<at 1 1 RCIM_Cubic>]}
       GreenCurve {keys [<at 0 0 RCIM_Cubic>,<at 0.5 1>,<at 1 0 RCIM_Cubic>]}
       RedCurve {keys [<at 0 1 RCIM_Cubic>,<at 0.5 0>,<at 1 0 RCIM_Cubic>]}
      }}
     CurveIndex=$Particles.NormalizedDistance
    )) {enabled=false}
  ]
  rendererClasses [&NiagaraSpriteRendererProperties]
  renderers [{rendererClass=&NiagaraSpriteRendererProperties}]
 }
 SnowGlobeGPU {
  simTarget=GPUComputeSim
  emitterUpdateScript [
   &EmitterState()
   &SpawnRate(SpawnRate=20, Use Spawn Probability=false)
   &SpawnRate(SpawnGroup=1, SpawnRate=100, Use Spawn Probability=false) {module=SpawnRate001}
   /SetVariables(Emitter.VelocityScale:NiagaraFloat=&Multiply_Float(A=0.015, B=&VectorLength(Vector=$Engine.Owner.Velocity))) {isSetParameters=true, module=SetVariables_#1}
  ]
  particleSpawnScript [
   &InitializeParticle(
    Color=&MakeLinearColorFromVectorAndFloat("Vector (RGB)"=<3 3 3>, "Float (Alpha)"=&RandomRangeFloat(Minimum=0.1, Maximum=1))
    Lifetime=&RandomRangeFloat(Minimum=2, Maximum=6)
    Module.Write Color=true
    Module.Write Lifetime=true
    Module.Write Mass=false
    Module.Write Material Random=true
    Module.Write Position=false
    Module.Write Scale=false
    Module.Write SpriteRotation=true
    Module.Write SpriteSize=true
    Position=&SimulationPosition()
    Sprite Rotation=&RandomRangeFloat(Minimum=-360, Maximum=360)
    Sprite Size=&MakeCustomVector2DFromBool(
     Bool=&SetBoolByIntComparison(A=&SpawnGroup(), B=0, Comparison Type=::NewEnumerator4)
     True Vector2D=&Vector2DFromFloat(Value=&RandomRangeFloat(Minimum=0.25, Maximum=&RandomRangeFloat(Minimum=0.5, Maximum=3)))
     False Vector2D=&Vector2DFromFloat(Value=&RandomRangeFloat(Minimum=0.2, Maximum=0.6))
    )
   )
   &SkeletalMeshLocation(
    Inherit Velocity=false
    MaskSpawnGroup=false
    Mesh Sampling Type=::NewEnumerator5
    Sampled Position Offset=&RandomRangeVector(Minimum=<-9 -9 -9>, Maximum=<9 9 9>)
    Sampled Position Offset Coordinate Space:ENiagaraMeshTransforms=::NewEnumerator1
    Skeletal Mesh=$User.SkeletalMesh
    Socket Sampling Mode=::NewEnumerator5
   )
  ]
  particleUpdateScript [
   &ParticleState()
   &SkeletalMeshLocation(
    Apply Sampled Orientation to Mesh Particles=true
    Apply To Particle Position=false
    Inherit Velocity=false
    Interpolate Mesh Sampling=false
    MaskSpawnGroup=false
    Mesh Sampling Type=::NewEnumerator5
    Sampled Position Offset Coordinate Space:ENiagaraMeshTransforms=::NewEnumerator1
    Skeletal Mesh=$User.SkeletalMesh
    Socket Sampling Mode=::NewEnumerator5
   ) {module=SkeletalMeshLocation001}
   &VortexVelocity(
    Module.Use Influence Falloff=false
    Velocity Amount:NiagaraFloat=&self:InterpolateFloat(
     Float to Interpolate:NiagaraFloat=&Lerp_Float(
      A=&RandomRangeFloat(Minimum=1, Maximum=3)
      B=&RandomRangeFloat(Minimum=10, Maximum=200)
      Alpha=&ClampFloat(Float=$Emitter.VelocityScale)
     )
     Initial Value:NiagaraFloat=2
     Rate Of Change:NiagaraFloat=3
    )
    Vortex Axis:Vector3f=&RandomRangeVector(Minimum=<-0.75 -0.75 1>, Maximum=<0.75 0.75 1>)
    Vortex Origin:Vector3f=$Output.SkeletalMeshLocation001.SampledPosition
   )
   &self:AddSmoothRandomVelocity(
    Rate Of Change=5
    Velocity:Vector3f=&Multiply_VectorByFloat(
     Vector=&Lerp_Vector(
      A=&RandomRangeVector(Minimum=<-5 -5 -6>, Maximum=<5 5 3>, Evaluation Type=::NewEnumerator1)
      B=&RandomRangeVector(Minimum=<-30 -30 -30>, Maximum=<30 30 30>, Evaluation Type=::NewEnumerator1)
      Alpha=&ClampFloat(Float=$Emitter.VelocityScale, Max=50)
     )
     Float=&ClampFloat(Float=&Multiply_Float(A=3, B=$Particles.NormalizedAge))
    )
   ) {module=AddSmoothRandomVelocity}
   &SolveForcesAndVelocity(Clamp Velocity=false, Limit Acceleration=false)
   &self:ConstrainToSphere(Sphere Radius:NiagaraFloat=&Multiply_Float(
     A=&MakeFloatFromVector(Vector=$Engine.Owner.Scale, Channel=::NewEnumerator0)
     B=&MakeCustomFloatFromBool(
      Bool=&SetBoolByIntComparison(A=&SpawnGroup(), B=0, Comparison Type=::NewEnumerator4)
      True Float=9
      False Float=&RandomRangeFloat(Minimum=4, Maximum=9.5)
     )
    )) {module=ConstrainToSphere}
   /SetVariables(Particles.SpriteRotation:NiagaraFloat=&Add_Float(
     A=$Particles.Initial.SpriteRotation
     B=&Lerp_Float(
      A=0
      B=&Multiply_Float(
       A=&RandomRangeFloat(Minimum=-2, Maximum=-1)
       B=&RandomRangeFloat(Minimum=150, Maximum=360)
      )
      Alpha=$Particles.NormalizedAge
     )
    )) {isSetParameters=true, module=SetVariables_#2}
   &ScaleSpriteSize(Scale Factor=&Multiply_Vector2(
     A=&MakeVector2D(
      X=1
      Y=&Lerp_Float(
       A=-1
       B=2
       Alpha=&self:InterpolateFloat(
        Float to Interpolate:NiagaraFloat=&RandomRangeFloat(Minimum=0, Maximum=1, Evaluation Type=::NewEnumerator1)
        Initial Value:NiagaraFloat=&RandomRangeFloat(Minimum=0.5, Maximum=1)
        Rate Of Change:NiagaraFloat=&RandomRangeFloat(Minimum=0.25, Maximum=0.75, Evaluation Type=::NewEnumerator1)
       )
      )
     )
     B=&Vector2DFromFloat(Value=&FloatFromCurve(
       FloatCurve=DataInterface() {propertyValues {Curve {keys [
           {arriveTangent=4.9809, interpMode=RCIM_Cubic, leaveTangent=4.9809, tangentMode=RCTM_User, time=0, value=0}
           {arriveTangent=0.9768, interpMode=RCIM_Cubic, leaveTangent=0.9768, tangentMode=RCTM_User, time=0.35, value=1}
           {arriveTangent=-1.0878, interpMode=RCIM_Cubic, leaveTangent=-1.0878, tangentMode=RCTM_User, time=0.65, value=1}
           {arriveTangent=-6.811, interpMode=RCIM_Cubic, leaveTangent=-6.811, tangentMode=RCTM_User, time=1, value=0}
          ]}}}
       Scale Curve=&MakeCustomFloatFromBool(
        Bool=&SetBoolByIntComparison(A=&SpawnGroup(), B=0, Comparison Type=::NewEnumerator4)
        True Float=1
        False Float=&Lerp_Float(
         A=1
         B=0
         Alpha=&ClampFloat(Float=&Multiply_Float(A=1000, B=&Add_Float(A=$Particles.NormalizedDistance, B=-0.999)))
        )
       )
      ))
    ))
   &Color(Color=&ColorFromCurve(
     ColorCurve=DataInterface() {propertyValues {
       AlphaCurve {keys [<at 0 1 RCIM_Cubic>]}
       BlueCurve {keys [<at 0 0 RCIM_Cubic>,<at 0.5 0>,<at 1 1 RCIM_Cubic>]}
       GreenCurve {keys [<at 0 0 RCIM_Cubic>,<at 0.5 1>,<at 1 0 RCIM_Cubic>]}
       RedCurve {keys [<at 0 1 RCIM_Cubic>,<at 0.5 0>,<at 1 0 RCIM_Cubic>]}
      }}
     CurveIndex=$Particles.NormalizedDistance
    )) {enabled=false}
  ]
  rendererClasses [&NiagaraSpriteRendererProperties]
  renderers [{rendererClass=&NiagaraSpriteRendererProperties}]
 }
 StarLights {
  emitterUpdateScript [
   &EmitterState()
   &SpawnBurst_Instantaneous(
    Spawn Count=1
    Spawn Probability=&MakeCustomFloatFromBool(
     Bool=&SetBoolByIntComparison(A=$System.LoopCount, B=0, Comparison Type=::NewEnumerator0)
     True Float=0
     False Float=1
    )
    Use Spawn Probability=true
   )
  ]
  particleSpawnScript [&InitializeParticle(
    Color=<rgba 1 0.9202 0.443 0.25>
    Module.Write Color=true
    Module.Write Lifetime=true
    Module.Write Mass=false
    Module.Write Material Random=true
    Module.Write Position=false
    Module.Write Scale=false
    Module.Write SpriteRotation=false
    Module.Write SpriteSize=true
    Position=&SimulationPosition()
    Sprite Size=&Vector2DFromFloat(Value=10)
   )]
  particleUpdateScript [
   &SkeletalMeshLocation(
    Bone Sampling Mode=::NewEnumerator4
    Inherit Velocity=false
    MaskSpawnGroup=false
    Sampled Position Offset=<0.5 -0.15 0>
    Sampled Position Offset Coordinate Space:ENiagaraMeshTransforms=::NewEnumerator3
    Skeletal Mesh=$User.SkeletalMesh
   )
   &ParticleState(Kill Particles When Lifetime Has Elapsed=false, Let Infinitely Lived Particles Die When Emitter Deactivates=true)
   &SolveForcesAndVelocity(Clamp Velocity=false, Limit Acceleration=false)
   &ScaleColor(
    Scale Alpha=&FloatFromCurve(
     FloatCurve=DataInterface() {propertyValues {Curve {keys [
         {interpMode=RCIM_Cubic, tangentMode=RCTM_User, time=0, value=1}
         {interpMode=RCIM_Cubic, tangentMode=RCTM_User, time=0.5, value=0.75}
         {interpMode=RCIM_Cubic, tangentMode=RCTM_User, time=1, value=1}
        ]}}}
     CurveIndex=Linked() {linkedVariable {name=NPC.FortniteNPC.TimeOfDay_0-1, type=NiagaraFloat}}
    )
    ScaleA=true
    ScaleRGB=false
   )
  ]
  rendererClasses [&NiagaraSpriteRendererProperties]
  renderers [{rendererClass=&NiagaraSpriteRendererProperties}]
 }
}
summary {
 systemName=NS_Backpack_HolidayLights
 emitters [
  {emitterName=SnowGlobe, rendererClasses [&NiagaraSpriteRendererProperties]}
  {emitterName=StarLights, rendererClasses [&NiagaraSpriteRendererProperties]}
  {emitterName=SnowGlobeGPU, simTarget=GPUComputeSim, rendererClasses [&NiagaraSpriteRendererProperties]}
 ]
 userVariables [
  User.bIsFrontEndPreview {type=NiagaraBool}
  User.PlayerSpeed {type=NiagaraFloat, defaultValue=0, description=""}
  User.SkeletalMesh {
   type=NiagaraDataInterfaceSkeletalMesh
   defaultValue {
    struct=NiagaraExt_VariableValue_DataInterface
    value {dataInterface=&self:NiagaraDataInterfaceSkeletalMesh_1}
   }
  }
 ]
}
