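# Masked, lit material for snowflakes on the Holiday Lights backpack accessory, with a grayscale texture as opacity mask and parameterized base and emissive colors.
# World position offset expands quads from centered UVs along camera X and Y axes, so they billboard at a set scale.
# It adds a clamped CameraOffset, whose world-position input is perturbed by time-driven sines rather than offset directly.
# The attributes pass through the accessories TechArt overrides function.
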Material /BRCosmetics/Accessories/FORT_Backpacks/M_MED_Holiday_Lights_Backpack/Materials/M_MED_HolidayLights_SnowFlakes
graph {
 properties {
  bUseMaterialAttributes=true
  bUsedWithMeshDeformer=true
  bUsedWithSkeletalMesh=true
  blendMode=BLEND_Masked
  pixelDepthOffsetMode=PDOM_Legacy
  usedShadingModels=Default Lit
 }
 nodes [
  Constant()
  n1 = TextureSample() {
   samplerType=SAMPLERTYPE_LinearGrayscale
   texture=/BRCosmetics/Characters/Player/Male/Medium/Bodies/M_MED_Holiday_Lights/Textures/T_MED_HolidayLights_SnowFlake
  }
  n2 = MaterialFunctionCall(Float2=Subtract(A=TextureCoordinate()) {constB=0.5}) {materialFunction=&BreakOutFloat2Components}
  n3 = Constant3Vector() {constant=<rgba 1 0 0 0>}
  n4 = MaterialFunctionCall(Float2=ScalarParameter() {defaultValue=4, parameterName=Scale}) {materialFunction=&BreakOutFloat2Components}
  n5 = Multiply(A=Multiply(A=@n2.R, B=Transform(None=@n3) {transformSourceType=TRANSFORMSOURCE_Camera}), B=@n4.R)
  n6 = Constant3Vector() {constant=<rgba 0 -1 0 0>}
  n7 = Multiply(A=Multiply(A=@n2.G, B=Transform(None=@n6) {transformSourceType=TRANSFORMSOURCE_Camera}), B=@n4.G)
  n8 = ScalarParameter() {defaultValue=3, group=Stars, parameterName="Stars Offset Amount (S)"}
  n9 = Time()
  n10 = ScalarParameter() {defaultValue=0.4, parameterName=T2}
  n11 = Multiply(A=Sine(None=Multiply(A=@n9, B=@n10)), B=ScalarParameter() {defaultValue=5, parameterName=Sine2})
  n12 = Sine(None=Multiply(A=@n9, B=ScalarParameter() {defaultValue=0.25, parameterName=T}))
  n13 = ScalarParameter() {defaultValue=30, parameterName=Sine}
  n14 = Add(
   A=WorldPosition().XYZ
   B=MaterialFunctionCall(X=@n11, Y=@n11, Z=Multiply(A=@n12, B=@n13)) {materialFunction=&MakeFloat3}.Result
  )
  n15 = SetMaterialAttributes!(
   MaterialAttributes
   Base Color=VectorParameter() {parameterName=SnowFlakeColor, defaultValue=<rgba 0.4674 0.6463 0.8646 0>}.RGB
   Specular=ScalarParameter() {defaultValue=1, parameterName=S}
   Roughness=ScalarParameter() {defaultValue=1, parameterName=R}
   Emissive Color=VectorParameter() {parameterName=SnowFlakeEmissive, defaultValue=<rgba 0.276 0.276 0.276 0>}.RGB
   Opacity Mask=@n1.R
   World Position Offset=Add(
    A=Add(A=@n5, B=@n7)
    B=MaterialFunctionCall(Offset Amount=@n8, World Position=@n14) {materialFunction=/Engine/Functions/Engine_MaterialFunctions02/WorldPositionOffset/CameraOffset}.Clamped Camera Offset
   )
  ) {attributeSetTypes [
    #1
    #2
    #3
    #4
    #5
    #6
   ]}
  MaterialFunctionCall(In=@n15) {materialFunction=&MF_TechArtOverrides_accessories}
 ]
}
