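# Physics asset for the Holiday Lights backpack cosmetic, defining simulation for its dynamic Christmas tree bone chain.
# Four small capsule bodies sit on the base bone and three numbered tree bones.
# Three constraints link each body to the previous one in a serial chain, with tight swing and twist limits of a few degrees, which suggests a slight, stiff sway rather than loose dangling.
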PhysicsAsset /BRCosmetics/Accessories/FORT_Backpacks/M_MED_Holiday_Lights_Backpack/Mesh/M_MED_Holiday_Lights_Pack_Physics
bodies {
 dyn_Christmas_treebase [{shapeType=Capsule, length=0.505, radius=0.505, rotation=<rot -0.003 -90 0>}]
 dyn_Christmas_treebase1 [{shapeType=Capsule, length=0.5, radius=0.2, center=<0.2 0 0>, rotation=<rot 90 0 0>}]
 dyn_Christmas_treebase2 [{shapeType=Capsule, length=0.2, radius=0.2, center=<0.5 0 0>, rotation=<rot 90 0 0>}]
 dyn_Christmas_treebase3 [{shapeType=Capsule, length=0.2, radius=0.2, center=<0.2 0 0>, rotation=<rot 90 0 0>}]
}
constraints [
 {bone1Name=dyn_Christmas_treebase2, bone2Name=dyn_Christmas_treebase1, swing1LimitDegrees=3, swing2LimitDegrees=3, twistLimitDegrees=10}
 {bone1Name=dyn_Christmas_treebase3, bone2Name=dyn_Christmas_treebase2, swing1LimitDegrees=5, swing2LimitDegrees=5, twistLimitDegrees=5}
 {bone1Name=dyn_Christmas_treebase1, bone2Name=dyn_Christmas_treebase, swing1LimitDegrees=2, swing2LimitDegrees=2, twistLimitDegrees=2}
]
